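# Translucent UI material for the Phantom special backing in the Challenges screen art, pairing a deep blue color with an animated, wispy glow.
# Two panning samples of a blurred noise texture multiply together to distort the UVs of two further panning noise layers.
# Their product modulates an exponential radial gradient centered on the bottom edge, plus a steady GlowAmount term.
# The result is boosted, clamped and scaled by OverallOpacity.
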Material /BRCosmetics/Athena/UI/Challenges/Art/Hyena/M_UI_Phantom_Special_Backing
graph {
 properties {blendMode=BLEND_Translucent, materialDomain=MD_UI, pixelDepthOffsetMode=PDOM_Legacy, usedShadingModels=Default Lit}
 nodes [
  Constant3Vector() {constant=<rgba 0 0.1595 0.4583 0>}
  n1 = MaterialFunctionCall(
   UVs=TextureCoordinate()
   CenterPosition=AppendVector(A=Constant() {r=0.5}, B=Constant() {r=1})
   Radius=Constant() {r=1}
   Density=Constant() {r=1.666}
  ) {materialFunction=&RadialGradientExponential}
  n2 = ScalarParameter() {defaultValue=0.2, parameterName=GlowAmount}
  n3 = ScalarParameter() {defaultValue=1.15, parameterName=Noise4_UTile}
  n4 = ScalarParameter() {defaultValue=1.15, parameterName=Noise4_VTile}
  n5 = TextureCoordinate()
  n6 = AppendVector(A=ScalarParameter() {defaultValue=-0.15, parameterName=Noise1_USpeed}, B=ScalarParameter() {defaultValue=0.075, parameterName=Noise1_VSpeed})
  n7 = AppendVector(A=ScalarParameter() {defaultValue=0.2, parameterName=Noise2_USpeed}, B=ScalarParameter() {defaultValue=0.025, parameterName=Noise2_VSpeed})
  n8 = DotProduct(
   A=TextureSample(UVs=Panner(Coordinate=@n5, Speed=@n6)) {samplerType=SAMPLERTYPE_Grayscale, texture=&LowResBlurredNoise}.R
   B=TextureSample(UVs=Panner(Coordinate=@n5, Speed=@n7)) {samplerType=SAMPLERTYPE_Grayscale, texture=&LowResBlurredNoise}.R
  )
  n9 = Add(A=Multiply(A=@n5, B=AppendVector(A=@n3, B=@n4)), B=Multiply(A=@n8, B=ScalarParameter() {defaultValue=0.5, parameterName=Noise4OffsetAmount}))
  n10 = AppendVector(A=ScalarParameter() {parameterName=Noise4_USpeed}, B=ScalarParameter() {defaultValue=0.05, parameterName=Noise4_VSpeed})
  n11 = ScalarParameter() {defaultValue=0.8, parameterName=Noise3_UTile}
  n12 = ScalarParameter() {defaultValue=0.8, parameterName=Noise3_VTile}
  n13 = Add(
   A=Multiply(A=@n5, B=AppendVector(A=@n11, B=@n12))
   B=Multiply(A=@n8, B=ScalarParameter() {defaultValue=0.25, parameterName=Noise3OffsetAmount})
  )
  n14 = AppendVector(A=ScalarParameter() {parameterName=Noise3_USpeed}, B=ScalarParameter() {defaultValue=0.15, parameterName=Noise3_VSpeed})
  n15 = DotProduct(
   A=TextureSample(UVs=Panner(Coordinate=@n9, Speed=@n10)) {samplerType=SAMPLERTYPE_Grayscale, texture=&LowResBlurredNoise}.R
   B=TextureSample(UVs=Panner(Coordinate=@n13, Speed=@n14)) {samplerType=SAMPLERTYPE_Grayscale, texture=&LowResBlurredNoise}.R
  )
  n16 = Multiply(A=Add(A=Multiply(A=@n1.RadialGradientExponential, B=@n2), B=DotProduct(A=@n1.RadialGradientExponential, B=@n15))) {constB=4}
  Saturate(None=Multiply(A=Saturate(None=@n16), B=ScalarParameter() {defaultValue=0.9, parameterName=OverallOpacity}))
 ]
}
