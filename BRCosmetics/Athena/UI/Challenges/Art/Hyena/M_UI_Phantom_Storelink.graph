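# Translucent UI material for the Phantom store-link art in the challenges UI, with animatable position and tint.
# RightTexture is sampled with halved V tiling, an offset whose U component blends via Anim_Translate, and center-based scaling.
# Its color blends toward FadeColor along a gradient in the bottom 30 percent, which AnimateTint removes as it reaches one.
# The texture's saturated alpha is passed through separately.
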Material /BRCosmetics/Athena/UI/Challenges/Art/Hyena/M_UI_Phantom_Storelink
graph {
 properties {blendMode=BLEND_Translucent, materialDomain=MD_UI, pixelDepthOffsetMode=PDOM_Legacy, usedShadingModels=Default Lit}
 nodes [
  n1 = TextureCoordinate() {vTiling=0.5}
  n2 = AppendVector(
   A=LinearInterpolate(
    A=ScalarParameter() {parameterName=U_Translate_Right}
    B=ScalarParameter() {defaultValue=0.025, parameterName=U_Translate_RightImage_Animated}
    Alpha=ScalarParameter() {parameterName=Anim_Translate}
   )
   B=ScalarParameter() {defaultValue=0.11, parameterName=V_Translate_RightImage}
  )
  n3 = ScalarParameter() {defaultValue=0.8, parameterName=U_Scale_RightImage}
  n4 = ScalarParameter() {defaultValue=0.8, parameterName=V_Scale_RightImage}
  n5 = TextureSampleParameter2D(UVs=MaterialFunctionCall(UVs=Add(A=@n1, B=@n2), Texture Scale=AppendVector(A=@n3, B=@n4)) {materialFunction=&ScaleUVsByCenter}.UVs) {parameterName=RightTexture, texture=&/T_UI_Phantom_Storelink}
  Reroute(None=Saturate(None=@n5.A))
  n6 = LinearInterpolate(A=Constant() {r=0.75}, B=Constant() {r=1}, Alpha=ScalarParameter() {parameterName=AnimateTint})
  n7 = Constant() {r=1}
  n8 = ComponentMask(None=TextureCoordinate()) {g=true}
  n9 = Saturate(None=Multiply(A=OneMinus(None=@n8)) {constB=3.333})
  Reroute(None=LinearInterpolate(
    A=VectorParameter() {parameterName=FadeColor, defaultValue=<rgba 0 0.1333 0.4 1>}.RGB
    B=@n5.RGB
    Alpha=LinearInterpolate(A=@n6, B=@n7, Alpha=@n9)
   ))
 ]
}
